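; Gameplay cue for the rez-out death effect on Lesser Demon NPCs, inheriting its core behavior from a parent cue class.
; It overrides the parent's hook for setting extra Niagara parameters on the spawned death FX, calling the parent first and then enlarging the effect's element size scale.
; BeginPlay, overlap and Tick events are empty, and the construction script only calls the parent.
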(event EventBeginPlay)

(event Collision|EventActorBeginOverlap
  (params (OtherActor :type Actor%20Object%20Reference)))

(event EventTick
  (params (DeltaSeconds :type Float%20%28single-precision%29)))

(event EventSetAdditionalNiagaraParametersonSpawnedFX
  (params (Spawned%20Death%20FX :type Niagara%20Particle%20System%20Component%20Object%20Reference))
  (|CallParentFunction Spawned%20Death%20FX)
  (Niagara|SetNiagaraVariable(Float) Spawned%20Death%20FX "ElementSizeScale" 1.25))

(fn ConstructionScript
  (|CallParentFunction))
